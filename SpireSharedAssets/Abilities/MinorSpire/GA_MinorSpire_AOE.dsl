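; Gameplay ability for the Minor Spire's area attack: a timed charge, then a downward beam and damaging pulse.
; On commit it sends an event, applies an active effect, and chains timers into charge and release.
; Release line-traces below the owner, stores the hit as AOE origin (also the custom source transform), fires cues, then applies an effect container to selected targets.
; Ending removes looping cues, the release timer and active effect.
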(event Ability|EventActivateAbility
  ; Activate Ability
  (|CallParentFunction)
  (Variables|Default|SetOwner (Ability|GetAvatarActorfromActorInfo))
  (bind _returnvalue (Ability|CommitAbility))
  (if _returnvalue
    (Utilities|Time|SetTimerbyFunctionName 0 "ChargeAttack" 1.0)
    (Ability|SendGameplayEventtoActor (Variables|Default|GetOwner) "(TagName=\"Ability.MinorSpire.AOE\")")
    (bind _returnvalue_2 (Ability|ApplyGameplayEffectToOwner (Variables|Default|GetGE_AOEAbilityActive)))
    (Variables|Default|SetAbilityActiveHandle _returnvalue_2)))

(event Custom|ChargeAttack
  ; Charge Attack
  (GameplayCue|AddGameplayCueOnActor(Looping) (Variables|Default|GetOwner) (Variables|GameplayCueTags|GetCharge%20Tag) (Ability|GameplayCue|MakeGameplayCueParameters 0.0))
  (bind _returnvalue (Utilities|Time|SetTimerbyFunctionName 0 "ReleaseAttack" (Variables|Default|GetChargeDuration)))
  (Variables|Default|SetReleaseAtkHandle _returnvalue))

(event Custom|ReleaseAttack
  (bind (_start _end) (CallFunction|GetAttackOrigin 2e+03))
  ; Release Attack
  (bind _outhit (Collision|LineTraceByChannel _start _end "TraceTypeQuery10" false (Utilities|Array|MakeArray (Variables|Default|GetOwner))))
  (Variables|Default|SetAOEOrigin (.Location (Collision|BreakHitResult _outhit)))
  (GameplayCue|AddGameplayCueOnActor(Looping) (Variables|Default|GetOwner) (Variables|GameplayCueTags|GetBeam%20Tag) (Ability|GameplayCue|MakeGameplayCueParameters :NormalizedMagnitude 0.0 :RawMagnitude 0.0 :Location "0,0,0" :Normal "0,0,0" :Instigator (Variables|Default|GetOwner) :GameplayEffectLevel 1 :AbilityLevel 1 :bReplicateLocationWhenUsingMinimalRepProxy false))
  (Utilities|FlowControl|Delay (Variables|Default|GetAttackTraceDelay))
  (GameplayCue|ExecuteGameplayCueOnActor(Burst) (Variables|Default|GetOwner) (Variables|GameplayCueTags|GetPulse%20Tag) (Ability|GameplayCue|MakeGameplayCueParameters :NormalizedMagnitude 0.0 :RawMagnitude 0.0 :Location (Variables|Default|GetAOEOrigin) :Normal "0,0,0" :GameplayEffectLevel 1 :AbilityLevel 1 :bReplicateLocationWhenUsingMinimalRepProxy false))
  (bind (_asynctaskproxy _targetdata _applicationtag) (Ability|Tasks|DoTargetSelection "DealDamage" "(TagName=\"Targeting.ApplicationTag.0\")" true)
    (:Targeted
      (Ability|ApplyGameplayEffectContainer _targetdata _applicationtag)
      (Utilities|FlowControl|Delay (Variables|Default|GetBeamDuration))
      (Ability|EndAbility))))

(event Ability|EventOnEndAbility
  (params (bWasCancelled :type Boolean))
  ; End Ability
  (|CallParentFunction bWasCancelled)
  (for _array_element (Utilities|Array|MakeArray (Variables|GameplayCueTags|GetBeam%20Tag) (Variables|GameplayCueTags|GetCharge%20Tag))
    (GameplayCue|RemoveGameplayCueOnActor(Looping) (Variables|Default|GetOwner) _array_element (Ability|GameplayCue|MakeGameplayCueParameters 0.0)))
  (Utilities|Time|ClearandInvalidateTimerbyHandle (Variables|Default|GetReleaseAtkHandle))
  (Ability|RemoveGameplayEffectFromOwnerWithHandle (Variables|Default|GetAbilityActiveHandle) 1))

(fn GetAttackOrigin
  (params (DistanceBelowOwner :type Float%20%28single-precision%29 :default 2e+03))
  (return (.x (Transformation|GetActorLocation (Variables|Default|GetOwner))) (.y (Transformation|GetActorLocation (Variables|Default|GetOwner))) (- (.z (Transformation|GetActorLocation (Variables|Default|GetOwner))) 237.0) (+ (* (Math|Vector|NegateVector (Transformation|GetActorUpVector (Variables|Default|GetOwner))) DistanceBelowOwner) (Transformation|GetActorLocation (Variables|Default|GetOwner)))))

(fn GetCustomAbilitySourceTransform
  (return (Variables|Default|GetAOEOrigin) "0, 0, 0" "1.000000,1.000000,1.000000"))
